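# Unlit glow material for the Athena island lake, built around a constant magenta HDR color.
# The color is multiplied by a shading-path switch that selects the Glow parameter by default and the lower GlowMobile parameter on mobile.
# A Tint parameter and a second multiply by WorldColor_EmissiveAmount appear with no visible link to that chain.
# The graph shows no connections to material outputs.
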Material /Game/Athena/Environments/Island/Materials/M_Athena_Lake_Glow
graph {
 properties {pixelDepthOffsetMode=PDOM_Legacy, shadingModel=MSM_Unlit, usedShadingModels=Unlit}
 nodes [
  Multiply(
   A=Constant3Vector() {constant=<rgba 3.9935 0.1549 4 0>}
   B=ShadingPathSwitch(Default=ScalarParameter() {defaultValue=6, parameterName=Glow}, Mobile=ScalarParameter() {defaultValue=3, parameterName=GlowMobile}) {desc=Migrated from FeatureLevelSwitch}
  )
  Multiply(B=ScalarParameter() {defaultValue=3, parameterName=WorldColor_EmissiveAmount})
  VectorParameter() {parameterName=Tint, defaultValue=<rgba 0.958 0.1102 1 0>}
 ]
}
