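# M_IslandBlend is a translucent, unlit material for the Athena island environment, apparently for soft color blending where geometry meets.
# A DepthFade node takes vertex alpha as its opacity and the Blend scalar as its fade distance, softening intersections with scene depth.
# A ColorTransition vector parameter supplies a tunable color, defaulting to over-bright magenta.
# The graph shows no links from either node to the material outputs.
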Material /Game/Athena/Environments/Island/Materials/M_IslandBlend
graph {
 properties {blendMode=BLEND_Translucent, pixelDepthOffsetMode=PDOM_Legacy, shadingModel=MSM_Unlit, usedShadingModels=Unlit}
 nodes [
  DepthFade(Opacity=VertexColor().A, FadeDistance=ScalarParameter() {defaultValue=269.5532, parameterName=Blend})
  VectorParameter() {parameterName=ColorTransition, defaultValue=<rgba 2 0 1.8913 0>}
 ]
}
