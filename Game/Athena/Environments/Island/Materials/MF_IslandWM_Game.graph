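# Material function for the Athena island material that outputs masks separating top, bottom, edge details, crystal and glow areas.
# Vertex or world-transformed input normals yield a sharpened downward-facing detail mask, while HeightLerp blends vertex normal Z with a height input for the island-scale top/bottom split.
# Actor-relative world Z drives a tunable edge gradient for rock details.
# Inverted CracksMask channels supply crystal (red) and boosted glow (red times green) masks.
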MaterialFunction /Game/Athena/Environments/Island/Materials/MF_IslandWM_Game
graph {nodes [
  n1 = StaticSwitchParameter(True=VertexNormalWS(), False=Transform(None=FunctionInput() {bUsePreviewValueAsDefault=true, inputName=N, previewValue=<0 0 1 1>})) {parameterName=UseVertexNormals}
  n2 = Constant3Vector() {constant=<rgba 0 0 -1 0>}
  n3 = Multiply(A=Subtract(A=DotProduct(A=@n1, B=@n2), B=Constant() {r=0.75}), B=Constant() {r=4})
  FunctionOutput(None=OneMinus(None=Saturate(None=@n3))) {description=Black Down White Up Details of rocks, outputName=MaskOut Bottom Details}
  n4 = VertexNormalWS()
  n5 = FunctionInput() {bUsePreviewValueAsDefault=true, inputName=up/Down Height Mask, inputType=FunctionInput_Scalar, previewValue=<0.5 0 0 1>}
  FunctionOutput(None=MaterialFunctionCall(Transition Phase=ComponentMask(None=@n4) {b=true}, Height Texture=@n5, Contrast=Constant() {r=2}) {materialFunction=&HeightLerp}.Alpha) {description=Black Down White Up Global Island Scale, outputName=Mask Top/Bottom, sortPriority=1}
  n6 = TextureSampleParameter2D() {parameterName=CracksMask, samplerType=SAMPLERTYPE_Masks, texture=/Game/Athena/Environments/Island/Textures/T_IslandPlug_M}
  n7 = OneMinus(None=@n6.R)
  n8 = OneMinus(None=@n6.G)
  FunctionOutput(None=Multiply(A=Multiply(A=@n7, B=@n8), B=Constant() {r=3})) {description=BitmapMask, outputName=GlowPartMask, sortPriority=2}
  n9 = Subtract(A=WorldPosition().XYZ, B=ActorPositionWS())
  n10 = OneMinus(None=Subtract(A=ComponentMask(None=@n9) {b=true}, B=ScalarParameter() {group="EdgeMask(Rocks)", parameterName=DetEdgeStartLine}))
  n11 = ScalarParameter() {defaultValue=850, group="EdgeMask(Rocks)", parameterName=DetEdgeBlur}
  FunctionOutput(None=Saturate(None=Divide(A=@n10, B=@n11))) {description=Gradient for details around the edge, outputName=GradientTopMask, sortPriority=2}
  FunctionOutput(None=@n7) {description=BitmapMask, outputName=Chrystal, sortPriority=2}
 ]}
