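# Layered surface material for an Athena island: rock on slopes, ground on upward faces.
# Height-lerps blend two-scale tinted rock with ground by normal direction, while vertex color drives the dark base (R), wet dirt (G), weeds (B) and corruption (inverted A).
# An animated, slope- and height-masked emissive glow follows alpha, adding bump-offset detail at higher quality.
# Forward and mobile paths simplify layers; MF_CreativeMoveTool wraps the result.
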Material /Game/Athena/Environments/Island/Materials/M_Island_2
graph {
 properties {bUseMaterialAttributes=true, bUsedWithInstancedStaticMeshes=true, pixelDepthOffsetMode=PDOM_Legacy, usedShadingModels=Default Lit}
 nodes [
  n1 = VertexColor()
  n2 = Multiply(A=TextureCoordinate(), B=Constant() {r=17.3647})
  n3 = TextureSample(UVs=@n2) {samplerSource=SSM_Wrap_WorldGroupSettings, texture=/Game/Athena/Environments/Island/Textures/Dirt_D}
  n4 = MaterialFunctionCall(Transition Phase=Multiply(A=@n1.G, B=Constant() {r=0.7266}), Height Texture=@n3.A, Contrast=Constant() {r=0.5}) {materialFunction=&HeightLerp}
  LinearInterpolate(Alpha=@n4.Lerp Alpha No Contrast) {constA=0.25, constB=0.1}
  n5 = Fresnel() {baseReflectFraction=0, exponent=1}
  n6 = TextureSampleParameter2D() {
   parameterName=M
   samplerSource=SSM_Wrap_WorldGroupSettings
   samplerType=SAMPLERTYPE_Masks
   texture=/Game/Athena/Environments/Island/Textures/T_Island2_M
  }
  n7 = BumpOffset(Coordinate=Multiply(A=TextureCoordinate(), B=Constant() {r=12}), Height=@n6.R, HeightRatioInput=Constant() {r=-0.8952}) {referencePlane=-0.25}
  n8 = Multiply(
   A=Constant() {r=0.5}
   B=Multiply(
    A=OneMinus(None=@n5)
    B=TextureSample(UVs=@n7) {samplerType=SAMPLERTYPE_Masks, texture=/Game/Athena/Prototype/Blueprints/Cube/Materials/T_DFHeightMask_Tile}.G
   )
  )
  n9 = TextureCoordinate()
  n10 = Constant() {r=0.45}
  n11 = TextureSampleParameter2D!(UVs=Rotator(Coordinate=Multiply(A=@n9, B=@n10), Time=Time()) {speed=0.125}, Level, Apply View MipBias) {
   constMipValue=3
   mipValueMode=TMVM_MipLevel
   parameterName=M
   samplerSource=SSM_Wrap_WorldGroupSettings
   samplerType=SAMPLERTYPE_Masks
   texture=&T_Island2_M
  }
  n12 = ComponentMask(None=Subtract(A=WorldPosition().XYZ, B=ObjectPositionWS())) {b=true}
  n13 = Constant() {r=611.0001}
  n14 = Saturate(None=Divide(A=Subtract(A=@n12, B=@n13), B=Constant() {r=-2000}))
  n15 = LinearInterpolate(A=Constant() {r=-0.5}, B=Constant() {r=1.15}, Alpha=DotProduct(A=VertexNormalWS(), B=Constant3Vector() {constant=<rgba 0 0 1 0>})) {constA=-1}
  n16 = OneMinus(None=@n15)
  n17 = VectorParameter() {group=Rocks, parameterName=ColorIsland, defaultValue=<rgba 0.2344 0.1401 0.0732 0>}
  n18 = ScalarParameter() {defaultValue=2000, group=Rocks, parameterName=ScaleRocks}
  n19 = MaterialFunctionCall!(TextureScale=@n18) {materialFunction=&/MF_IslandRocks_Game}
  n20 = MaterialFunctionCall(TextureScale=Divide(A=@n18)) {materialFunction=&MF_IslandRocks_Game}
  n21 = Constant() {r=0.33}
  n22 = MaterialFunctionCall(
   A=Multiply(
    A=@n17.RGB
    B=MaterialFunctionCall(A=@n19.WS RGB, B=@n20.WS RGB, Transition Phase=@n21, Height Texture=@n19.WS RGB) {materialFunction=&HeightLerp}.Results
   )
   B=@n17.RGB
   Transition Phase=Constant() {r=0.5}
   Height Texture=@n6.R
   Contrast=Constant() {r=0.3}
  ) {materialFunction=&HeightLerp}
  n23 = LinearInterpolate(A=Constant() {r=-0.5}, B=Constant() {r=2}, Alpha=@n22.Alpha)
  n24 = Multiply(
   A=Multiply(A=@n11.B, B=Multiply(A=@n14, B=@n16))
   B=Multiply(A=VectorParameter() {parameterName=TintGlow, defaultValue=<rgba 2.6496 0 4 1>}.RGB, B=Saturate(None=@n23))
  )
  n25 = Multiply(A=Add(A=Saturate(None=@n8), B=Constant() {r=0.4}), B=@n24)
  n26 = QualitySwitch(Default=@n25, Low=@n24, High=@n25, Medium=@n24)
  n27 = Multiply(A=TextureCoordinate(), B=ScalarParameter() {defaultValue=25, parameterName=MudTextureTiling})
  n28 = TextureSample(UVs=@n27) {samplerSource=SSM_Wrap_WorldGroupSettings, samplerType=SAMPLERTYPE_Normal, texture=/Game/Packages/DS_Fortnight/WhiteRockCustom_N}
  n29 = SetMaterialAttributes!(
   MaterialAttributes
   Base Color=Constant3Vector() {constant=<rgba 0.13 0.0734 0.0482 1>}
   Roughness=Constant() {r=0.5}
   Specular=Constant() {r=0.07}
   Normal=@n28.RGB
  ) {attributeSetTypes [
    #1
    #2
    #3
    #4
   ]}
  n30 = LinearInterpolate(A=VectorParameter() {group=Rocks, parameterName=ColorIsland_AO, defaultValue=<rgba 0.1667 0.0515 0.0035 0>}.RGB, B=@n22.Results, Alpha=@n6.G)
  n31 = LinearInterpolate(A=Constant() {r=0.1}, B=Constant() {r=0.5}, Alpha=@n6.G)
  n32 = LinearInterpolate(A=Constant() {r=1}, B=Constant() {r=0.55}, Alpha=@n6.G)
  n33 = TextureSampleParameter2D!(UVs, Apply View MipBias) {
   parameterName=N
   samplerSource=SSM_Wrap_WorldGroupSettings
   samplerType=SAMPLERTYPE_Normal
   texture=/Game/Athena/Environments/Island/Textures/T_Island2_N
  }
  n34 = TextureSampleParameter2D() {
   parameterName=N
   samplerSource=SSM_Wrap_WorldGroupSettings
   samplerType=SAMPLERTYPE_Normal
   texture=&T_Island2_N
  }
  n35 = LinearInterpolate(
   A=LinearInterpolate(A=Constant3Vector() {constant=<rgba 0 0 1 0>}, B=@n19.WS N)
   B=@n19.WS N
   Alpha=MaterialFunctionCall!(up/Down Height Mask, N=@n34.RGB) {materialFunction=&/MF_IslandWM_Game}.MaskOut Bottom Details
  )
  n36 = MaterialFunctionCall(
   BaseNormal=@n33.RGB
   AdditionalNormal=MaterialFunctionCall(Normal=ShadingPathSwitch(Default=@n35, Deferred=@n35, Forward=@n19.WS N, Mobile=@n19.WS N), Flatness=Constant() {r=0.5}) {materialFunction=&FlattenNormal}.Result
  ) {materialFunction=&BlendAngleCorrectedNormals}
  n37 = Constant() {r=5}
  n38 = TextureSample(UVs=@n27) {samplerSource=SSM_Wrap_WorldGroupSettings, texture=/Game/Packages/DS_Fortnight/WhiteRockCustom_D}
  n39 = TextureSample(UVs=@n2) {samplerSource=SSM_Wrap_WorldGroupSettings, samplerType=SAMPLERTYPE_Normal, texture=/Game/Packages/DS_Fortnight/ExperimentalTextures/Dirt_N}
  n40 = Constant3Vector() {constant=<rgba -1 -1 1 0>}
  n41 = ComponentMask(None=WorldPosition().XYZ) {g=true, r=true}
  n42 = Constant() {r=250}
  n43 = TextureSample(UVs=Panner(Coordinate=Divide(A=@n41, B=@n42)) {speedX=0.1, speedY=0.1}) {samplerSource=SSM_Wrap_WorldGroupSettings, samplerType=SAMPLERTYPE_Normal, texture=/Game/Global/Textures/Default/Noise/T_SwirlNoise_N}
  n44 = MaterialFunctionCall(Normal=@n43.RGB, Flatness=Constant() {r=0.99}) {materialFunction=&FlattenNormal}
  n45 = Constant3Vector() {constant=<rgba 0 0 1 0>}
  n46 = Multiply(A=@n3.RGB, B=Constant3Vector() {constant=<rgba 0.276 0.2461 0.1642 1>}) {constB=0.5}
  n47 = Fresnel() {exponent=3}
  n48 = SetMaterialAttributes!(
   MaterialAttributes
   Base Color=LinearInterpolate(A=@n46, B=LinearInterpolate(A=@n46, Alpha=@n47) {constB=0}, Alpha=@n4.Alpha)
   Normal=LinearInterpolate(A=Multiply(A=@n39.RGB, B=@n40), B=QualitySwitch(Default=@n44.Result, Low=@n45, High=@n44.Result, Medium=@n44.Result), Alpha=@n4.Alpha) {constA=0.35}
   Roughness=LinearInterpolate(A=Constant() {r=0.55}, B=Constant() {r=0.15}, Alpha=@n4.Alpha) {constA=0.35}
   Specular=LinearInterpolate(A=Constant() {r=0.35}, B=Constant() {r=0.1}, Alpha=@n4.Alpha) {constA=0.35}
  ) {attributeSetTypes [
    #1
    #4
    #2
    #3
   ]}
  n49 = Constant() {r=0.3}
  n50 = Saturate(None=LinearInterpolate(A=Constant() {r=-1}, B=Constant() {r=1}, Alpha=@n38.A))
  n51 = Multiply(A=VectorParameter() {group=Mud, parameterName=MudTintDry, defaultValue=<rgba 0.5677 0.4705 0.4021 0>}.RGB, B=@n38.RGB)
  n52 = SetMaterialAttributes!(
   MaterialAttributes
   Base Color=@n51
   Specular=LinearInterpolate(A=Constant() {r=0.05}, B=Constant() {r=0.5}, Alpha=@n50) {constB=0.4}
   Roughness=LinearInterpolate(A=Constant() {r=0.99}, B=Constant() {r=0.65}, Alpha=@n50)
   Normal=@n28.RGB
  ) {attributeSetTypes [
    #1
    #3
    #2
    #4
   ]}
  n53 = ShadingPathSwitch(Default=@n52, Deferred=@n52, Forward=@n52, Mobile=@n52)
  n54 = BlendMaterialAttributes(
   A=@n53
   B=@n48
   Alpha=MaterialFunctionCall(Transition Phase=Add(A=@n49, B=@n1.G), Height Texture=@n3.A, Contrast=Constant() {r=0.2}) {materialFunction=&HeightLerp}.Alpha
  )
  n55 = Divide(A=ComponentMask(None=WorldPosition().XYZ) {g=true, r=true}, B=Constant() {r=512})
  n56 = TextureSample(UVs=Multiply(A=@n55, B=Constant() {r=2})) {samplerSource=SSM_Wrap_WorldGroupSettings, samplerType=SAMPLERTYPE_Normal, texture=/Game/Athena/Environments/Island/Textures/T_IslandGreen_N}
  n57 = Constant() {r=0.3}
  n58 = TextureSample(UVs=@n55) {samplerSource=SSM_Wrap_WorldGroupSettings, texture=/Game/Athena/Environments/Island/Textures/T_IslandGreen_D}
  n59 = Multiply(A=@n58.R, B=ScalarParameter() {defaultValue=1, parameterName=WeedSpecular})
  n60 = ScalarParameter() {defaultValue=0.05, parameterName=WeedSpecConstrast}
  n61 = Multiply(A=@n58.RGB, B=Constant3Vector() {constant=<rgba 0.3013 0.4531 0.1224 0>})
  n62 = Constant3Vector() {constant=<rgba 0.026 0.0167 0.004 0>}
  n63 = LinearInterpolate(
   A=@n61
   B=LinearInterpolate(A=LinearInterpolate(A=@n62, B=@n61) {constA=0.1}, B=@n62, Alpha=Fresnel() {exponent=3}) {constB=0}
   Alpha=@n4.Lerp Alpha No Contrast
  )
  n64 = SetMaterialAttributes!(
   MaterialAttributes
   Base Color=@n63
   Normal=LinearInterpolate(
    A=MaterialFunctionCall(Normal=@n56.RGB, Flatness=@n57) {materialFunction=&FlattenNormal}.Result
    B=Constant3Vector() {constant=<rgba 0 0 1 0>}
    Alpha=@n4.Lerp Alpha No Contrast
   )
   Roughness=LinearInterpolate(A=Constant() {r=1}, Alpha=@n4.Lerp Alpha No Contrast) {constA=0.55, constB=0.15}
   Specular=Saturate(None=MaterialFunctionCall(In=@n59, Contrast=@n60) {materialFunction=&CheapContrast}.Result)
  ) {attributeSetTypes [
    #1
    #4
    #2
    #3
   ]}
  n65 = VertexColor()
  n66 = Constant() {r=0.85}
  n67 = BlendMaterialAttributes(
   A=ShadingPathSwitch(Default=@n54, Deferred=@n54, Forward=@n53, Mobile=@n53)
   B=ShadingPathSwitch(Default=@n64, Deferred=@n64, Forward=@n64, Mobile=@n64)
   Alpha=MaterialFunctionCall(Transition Phase=Multiply(A=@n65.B, B=@n66), Height Texture=Subtract(A=@n3.A, B=@n58.A), Contrast=Constant() {r=4}) {materialFunction=&HeightLerp}.Alpha
  )
  n68 = MaterialFunctionCall!(ScaleUV=ScalarParameter() {defaultValue=0.75}) {materialFunction=/Game/Athena/Environments/Landscape/MaterialFunctions/Corruption/MF_Athena_Corruption_01}
  n69 = GetMaterialAttributes(None=@n68.Result) {attributeGetTypes [
    #5
    #2
    #3
    #1
   ]}
  n70 = MaterialFunctionCall(Transition Phase=OneMinus(None=VertexColor().A), Height Texture=@n69.Opacity, Contrast=ScalarParameter() {defaultValue=5, parameterName=Cor}) {materialFunction=&HeightLerp}
  n71 = BlendMaterialAttributes(
   A=@n29
   B=BlendMaterialAttributes(
    A=MakeMaterialAttributes!(
     BaseColor=@n30
     Metallic
     Specular=@n31
     Roughness=@n32
     Anisotropy
     EmissiveColor
     Opacity
     OpacityMask
     Normal=@n36.Result
     Tangent
     WorldPositionOffset
     SubsurfaceColor
     ClearCoat
     ClearCoatRoughness
     AmbientOcclusion
     Refraction
     CustomizedUVs_0
     CustomizedUVs_1
     CustomizedUVs_2
     CustomizedUVs_3
     CustomizedUVs_4
     CustomizedUVs_5
     CustomizedUVs_6
     CustomizedUVs_7
     PixelDepthOffset
     ShadingModel
     Displacement
    )
    B=BlendMaterialAttributes(A=@n67, B=@n68.Result, Alpha=@n70.Alpha)
    Alpha=MaterialFunctionCall(Transition Phase=@n15, Height Texture=@n38.A, Contrast=@n37) {materialFunction=&HeightLerp}.Alpha
   )
   Alpha=MaterialFunctionCall(Transition Phase=VertexColor().R, Height Texture=OneMinus(None=@n38.A), Contrast=Constant() {r=5}) {materialFunction=&HeightLerp}.Alpha
  )
  n72 = OneMinus(None=VertexColor().A)
  n73 = OneMinus(None=@n19.WS A)
  n74 = Constant() {r=-0.02}
  MaterialFunctionCall(MA In=BlendMaterialAttributes(
    A=@n71
    B=SetMaterialAttributes!(MaterialAttributes=@n71, Emissive Color=@n26) {attributeSetTypes [#6]}
    Alpha=MaterialFunctionCall(Transition Phase=@n72, Height Texture=@n73, Contrast=@n74) {materialFunction=&HeightLerp}.Alpha
   )) {materialFunction=&MF_CreativeMoveTool}
  Reroute(None=@n3.R)
  SetMaterialAttributes!(MaterialAttributes, Base Color=@n63) {attributeSetTypes [#1]}
  SetMaterialAttributes!(MaterialAttributes, Base Color=@n51, Normal=@n28.RGB) {attributeSetTypes [#1, #4]}
  SetMaterialAttributes!(MaterialAttributes, Base Color=@n51) {attributeSetTypes [#1]}
 ]
}
