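# Material function for island environment materials that outputs two radial ground masks built from sphere masks.
# The default output forms an object-relative ring, multiplying a large sphere around the object position by the inverse of a smaller, offset sphere.
# "WS Sphere DonutMask" saturates one world-space sphere driven by WorldSphere, Rad and Soft parameters.
# Despite that name, its inner-cutout multiply connects to no output, so that mask has no hole.
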MaterialFunction /Game/Athena/Environments/Island/Materials/MF_IslandRadialGroundMask
graph {nodes [
  Constant() {r=3222.0984}
  n1 = Subtract(A=WorldPosition().XYZ, B=ObjectPositionWS())
  n2 = Constant3Vector() {constant=<rgba 0 0 0 1>}
  n3 = Constant() {desc=4278.969238, r=6500.0718}
  n4 = Constant() {r=-0.019}
  n5 = SphereMask(A=@n1, B=Constant3Vector() {constant=<rgba -93.5197 -341.1215 0 1>}, Radius=Constant() {r=3222.0984}, Hardness=Constant() {r=0.8646})
  FunctionOutput(None=Multiply(A=SphereMask(A=@n1, B=@n2, Radius=@n3, Hardness=@n4), B=OneMinus(None=@n5)))
  n6 = WorldPosition()
  n7 = VectorParameter() {parameterName=WorldSphere, defaultValue=<rgba 642.6195 -23460.1562 4699.1377 1>}
  n8 = ScalarParameter() {defaultValue=4521.269, desc=4278.969238, parameterName=Rad}
  n9 = SphereMask(A=@n6.XYZ, B=@n7.RGB, Radius=@n8, Hardness=ScalarParameter() {defaultValue=-5.3004, parameterName=Soft})
  FunctionOutput(None=Saturate(None=@n9)) {outputName=WS Sphere DonutMask}
  n10 = Divide(A=@n8)
  n11 = Constant() {r=0.8646}
  Multiply(A=@n9, B=OneMinus(None=SphereMask(A=@n6.XYZ, B=@n7.RGB, Radius=@n10, Hardness=@n11)))
 ]}
