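# Master material for the Athena island, switching statically between a ground-topped rock variant and a glowing rock variant.
# Ground layers mud, panning-normal wet dirt, green cover, corruption and dark soil through vertex-colour height blends, masked over two-scale rock.
# The glow variant uses object-relative sphere masks, noise and parallax, with optional pulsing that turns erratic via the Cube collection's instability parameter.
# Shading-path and quality switches simplify cheaper platforms, and output passes through MF_CreativeMoveTool.
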Material /Game/Athena/Environments/Island/Materials/M_Island_Game
graph {
 properties {
  bUseMaterialAttributes=true
  bUsedWithInstancedStaticMeshes=true
  bUsedWithNanite=true
  pixelDepthOffsetMode=PDOM_Legacy
  usedShadingModels=Default Lit
 }
 nodes [
  Constant() {r=1}
  n1 = Multiply(A=TextureCoordinate(), B=ScalarParameter() {defaultValue=17.3647, parameterName=DirtScaling})
  n2 = Divide(A=ComponentMask(None=WorldPosition().XYZ) {g=true, r=true}, B=Constant() {r=250})
  n3 = MaterialFunctionCall(
   Normal=TextureSample(UVs=Panner(Coordinate=@n2) {speedX=0.1, speedY=0.1}) {samplerType=SAMPLERTYPE_Normal, texture=/Game/Global/Textures/Default/Noise/T_SwirlNoise_N}.RGB
   Flatness=Constant() {r=0.99}
  ) {materialFunction=&FlattenNormal}
  n4 = VertexColor()
  n5 = TextureSample(UVs=@n1) {texture=/Game/Athena/Environments/Island/Textures/Dirt_D}
  n6 = Constant() {r=12}
  n7 = Constant() {r=-29.4702}
  n8 = MaterialFunctionCall() {materialFunction=&/MF_IslandRadialGroundMask}
  n9 = MaterialFunctionCall(
   Transition Phase=Multiply(A=@n4.G, B=Constant() {r=0.7266})
   Height Texture=@n5.A
   Contrast=Saturate(None=LinearInterpolate(A=@n6, B=@n7, Alpha=@n8.Result))
  ) {materialFunction=&HeightLerp}
  n10 = LinearInterpolate(
   A=Multiply(
    A=TextureSample(UVs=@n1) {samplerType=SAMPLERTYPE_Normal, texture=/Game/Packages/DS_Fortnight/ExperimentalTextures/Dirt_N}.RGB
    B=Constant3Vector() {constant=<rgba -1 -1 1 0>}
   )
   B=QualitySwitch(Default=@n3.Result, Low=Constant3Vector() {constant=<rgba 0 0 1 0>}, High=@n3.Result, Medium=@n3.Result)
   Alpha=@n9.Alpha
  ) {constA=0.35}
  n11 = Multiply(A=@n5.RGB, B=Constant3Vector() {constant=<rgba 0.3802 0.339 0.2262 0>}) {constB=0.5}
  n12 = Fresnel() {exponent=3}
  n13 = SetMaterialAttributes!(
   MaterialAttributes
   Base Color=LinearInterpolate(A=@n11, B=LinearInterpolate(A=@n11, Alpha=@n12) {constB=0}, Alpha=@n9.Alpha)
   Normal=@n10
   Roughness=LinearInterpolate(A=Constant() {r=0.55}, B=Constant() {r=0.15}, Alpha=@n9.Alpha) {constA=0.35}
   Specular=LinearInterpolate(A=Constant() {r=0.35}, B=Constant() {r=0.1}, Alpha=@n9.Alpha) {constA=0.35}
  ) {attributeSetTypes [
    #1
    #2
    #3
    #4
   ]}
  n14 = Constant() {r=0.3}
  n15 = Multiply(A=TextureCoordinate(), B=ScalarParameter() {defaultValue=15.1741, parameterName=MudTextureTiling})
  n16 = TextureSample(UVs=@n15) {texture=/Game/Packages/DS_Fortnight/WhiteRockCustom_D}
  n17 = Saturate(None=LinearInterpolate(A=Constant() {r=-1}, B=Constant() {r=1}, Alpha=@n16.A))
  n18 = LinearInterpolate(A=Constant() {r=0.05}, B=Constant() {r=0.5}, Alpha=@n17) {constB=0.4}
  n19 = MaterialFunctionCall() {materialFunction=&MF_IslandRadialGroundMask}
  n20 = LinearInterpolate(A=Constant() {r=0.99}, B=Constant() {r=0.65}, Alpha=@n17)
  n21 = Multiply(
   A=LinearInterpolate(
    A=VectorParameter() {group=Mud, parameterName=MudTint, defaultValue=<rgba 0.2396 0.2225 0.2034 0>}.RGB
    B=VectorParameter() {group=Mud, parameterName=MudTintDry, defaultValue=<rgba 0.3438 0.2234 0.1387 0>}.RGB
    Alpha=@n19.Result
   )
   B=@n16.RGB
  )
  n22 = TextureSample(UVs=@n15) {samplerType=SAMPLERTYPE_Normal, texture=/Game/Packages/DS_Fortnight/WhiteRockCustom_N}
  n23 = SetMaterialAttributes!(
   MaterialAttributes
   Base Color=@n21
   Specular=LinearInterpolate(A=Multiply(A=@n18) {constB=0.5}, B=@n18, Alpha=@n19.Result) {constB=0.7}
   Roughness=LinearInterpolate(A=Multiply(A=@n20) {constB=0.5}, B=@n20, Alpha=@n19.Result) {constB=0.7}
   Normal=@n22.RGB
  ) {attributeSetTypes [
    #1
    #4
    #3
    #2
   ]}
  n24 = ShadingPathSwitch(
   Default=@n23
   Deferred=@n23
   Forward=SetMaterialAttributes!(MaterialAttributes, Base Color=@n21, Normal=@n22.RGB) {attributeSetTypes [#1, #2]}
   Mobile=SetMaterialAttributes!(MaterialAttributes, Base Color=@n21, Specular=Constant()) {attributeSetTypes [#1, #4]}
  )
  n25 = BlendMaterialAttributes(
   A=@n24
   B=@n13
   Alpha=MaterialFunctionCall(
    Transition Phase=Add(A=@n14, B=@n4.G)
    Height Texture=@n5.A
    Contrast=ScalarParameter() {defaultValue=0.2, parameterName=zzzz, sliderMin=0.0284}
   ) {materialFunction=&HeightLerp}.Alpha
  )
  GetMaterialAttributes(None=@n25) {attributeGetTypes [#2]}
  n26 = WorldPosition()
  n27 = TextureSample(UVs=Divide(A=ComponentMask(None=@n26.XYZ) {g=true, r=true}, B=Constant() {r=1500})) {samplerType=SAMPLERTYPE_Masks, texture=/Game/Athena/Environments/Island/Textures/T_FNBlendmask}
  n28 = SetMaterialAttributes!(
   MaterialAttributes
   Base Color=LinearInterpolate(A=Constant3Vector() {constant=<rgba 0.026 0.0084 0.0065 0>}, B=Constant3Vector() {constant=<rgba 0.082 0.0491 0.033 0>}, Alpha=@n27.R)
   Roughness=Constant() {r=0.5}
   Specular=Constant() {r=0.07}
   Normal=@n10
  ) {attributeSetTypes [
    #1
    #3
    #4
    #2
   ]}
  n29 = MaterialFunctionCall!(ScaleUV=Constant() {r=0.5}) {materialFunction=/Game/Athena/Environments/Landscape/MaterialFunctions/Corruption/MF_Athena_Corruption_01}
  n30 = GetMaterialAttributes(None=@n29.Result) {attributeGetTypes [
    #5
    #3
    #4
    #1
   ]}
  n31 = Saturate(None=MaterialFunctionCall!(Cutoff Value=Constant() {r=1.0302}, Lerp Value=@n30.Opacity, Gradient=Constant() {r=0.1773}) {materialFunction=/Engine/Functions/Engine_MaterialFunctions02/SmoothThreshold}.Result)
  n32 = Constant3Vector() {constant=<rgba 1.2 1.1125 1.1885 0>}
  n33 = SetMaterialAttributes!(
   MaterialAttributes=@n30.MaterialAttributes
   Roughness=LinearInterpolate(A=Multiply(A=@n30.Roughness) {constB=0.35}, B=Multiply(A=@n30.Roughness) {constB=0.75}, Alpha=@n31)
   Specular=LinearInterpolate(A=Multiply(A=@n30.Specular) {constB=0.25}, B=Multiply(A=@n30.Specular), Alpha=@n31)
   Base Color=LinearInterpolate(
    A=VectorParameter() {parameterName=CorruptGround, defaultValue=<rgba 0.0781 0.0615 0.048 0>}.RGB
    B=Multiply(A=@n30.Base Color, B=@n32)
    Alpha=@n31
   )
  ) {attributeSetTypes [#3, #4, #1]}
  n34 = GetMaterialAttributes(None=@n29.Result) {attributeGetTypes [#1, #2]}
  n35 = SetMaterialAttributes!(MaterialAttributes, Base Color=@n34.Base Color, Normal=@n34.Base Color) {attributeSetTypes [#1, #2]}
  n36 = Divide(A=ComponentMask(None=WorldPosition().XYZ) {g=true, r=true}, B=Constant() {r=512})
  n37 = TextureSample(UVs=Multiply(A=@n36, B=Constant() {r=2})) {samplerType=SAMPLERTYPE_Normal, texture=/Game/Athena/Environments/Island/Textures/T_IslandGreen_N}
  n38 = Constant() {r=0.3}
  n39 = TextureSample(UVs=@n36) {texture=/Game/Athena/Environments/Island/Textures/T_IslandGreen_D}
  n40 = Multiply(A=@n39.RGB, B=Constant3Vector() {constant=<rgba 0.3013 0.4531 0.1224 0>})
  n41 = Constant3Vector() {constant=<rgba 0.026 0.0167 0.004 0>}
  n42 = LinearInterpolate(
   A=@n40
   B=LinearInterpolate(A=LinearInterpolate(A=@n41, B=@n40) {constA=0.1}, B=@n41, Alpha=Fresnel() {exponent=3}) {constB=0}
   Alpha=@n9.Lerp Alpha No Contrast
  )
  n43 = SetMaterialAttributes!(
   MaterialAttributes
   Base Color=@n42
   Normal=LinearInterpolate(
    A=MaterialFunctionCall(Normal=@n37.RGB, Flatness=@n38) {materialFunction=&FlattenNormal}.Result
    B=Constant3Vector() {constant=<rgba 0 0 1 0>}
    Alpha=@n9.Lerp Alpha No Contrast
   )
   Roughness=LinearInterpolate(A=Constant() {r=1}, Alpha=@n9.Lerp Alpha No Contrast) {constA=0.55, constB=0.15}
   Specular=LinearInterpolate(Alpha=@n9.Lerp Alpha No Contrast) {constA=0.25, constB=0.1}
  ) {attributeSetTypes [
    #1
    #2
    #3
    #4
   ]}
  n44 = SetMaterialAttributes!(MaterialAttributes, Base Color=@n42, Specular=Constant()) {attributeSetTypes [#1, #4]}
  n45 = Multiply(A=VertexColor().B, B=Constant() {r=0.85})
  n46 = Subtract(A=@n27.R, B=@n39.A)
  n47 = Constant() {r=4}
  n48 = BlendMaterialAttributes(
   A=BlendMaterialAttributes(
    A=ShadingPathSwitch(Default=@n25, Deferred=@n25, Forward=@n24, Mobile=@n24)
    B=ShadingPathSwitch(Default=@n43, Deferred=@n43, Forward=@n44, Mobile=@n44)
    Alpha=MaterialFunctionCall(Transition Phase=@n45, Height Texture=@n46, Contrast=@n47) {materialFunction=&HeightLerp}.Alpha
   )
   B=ShadingPathSwitch(Default=@n33, Deferred=@n33, Forward=@n35, Mobile=@n35)
   Alpha=MaterialFunctionCall(
    Transition Phase=VertexColor().R
    Height Texture=GetMaterialAttributes(None=@n29.Result) {attributeGetTypes [#5]}.Opacity
    Contrast=Constant() {r=4}
   ) {materialFunction=&HeightLerp}.Alpha
  )
  n49 = BlendMaterialAttributes(
   A=@n28
   B=@n48
   Alpha=MaterialFunctionCall(Transition Phase=VertexColor().A, Height Texture=@n5.A, Contrast=ScalarParameter() {defaultValue=13.8918, parameterName=sssa}) {materialFunction=&HeightLerp}.Alpha
  )
  n50 = QualitySwitch(Default=@n49, Low=@n48, High=@n49, Medium=@n48)
  n51 = ShadingPathSwitch(Default=@n50, Deferred=@n50, Forward=@n48, Mobile=@n48)
  n52 = ScalarParameter() {defaultValue=-3, parameterName=Rock_MaskContrast_Min}
  n53 = ScalarParameter() {defaultValue=1, parameterName=Rock_MaskContrast_Max}
  n54 = MaterialFunctionCall!(up/Down Height Mask=Min(A=@n5.A, B=@n16.A), N) {materialFunction=&/MF_IslandWM_Game}
  n55 = StaticSwitchParameter(True=Clamp(None=LinearInterpolate(A=@n52, B=@n53, Alpha=@n54.Mask Top/Bottom)), False=@n54.Mask Top/Bottom) {parameterName=Use_MaskContrast}
  n56 = ScalarParameter() {defaultValue=3, group=Rocks, parameterName=ScaleRocks}
  n57 = MaterialFunctionCall!(TextureScale=@n56) {materialFunction=&/MF_IslandRocks_Game}
  n58 = MaterialFunctionCall(TextureScale=Multiply(B=@n56) {constA=3.5}) {materialFunction=&MF_IslandRocks_Game}
  n59 = MaterialFunctionCall(N=TextureSampleParameter2D() {parameterName=N, samplerType=SAMPLERTYPE_Normal, texture=/Game/Athena/Environments/Island/Textures/T_IslandTop_N}.RGB) {materialFunction=&MF_IslandWM_Game}
  n60 = MaterialFunctionCall(
   A=Multiply(A=ScalarParameter() {defaultValue=0.5, group=Rocks, parameterName=BaseBodyBrightness}, B=@n57.RGB)
   B=Multiply(A=ScalarParameter() {defaultValue=2, group=Rocks, parameterName=EdgeBrightness}, B=@n58.RGB)
   Transition Phase=Constant() {r=0.33}
   Height Texture=Subtract(A=@n57.A, B=@n59.GradientTopMask)
  ) {materialFunction=&HeightLerp}
  n61 = Multiply(A=@n60.Results, B=LinearInterpolate(A=Constant() {r=1}, B=Constant() {r=3}, Alpha=@n59.MaskOut Bottom Details))
  n62 = LinearInterpolate(
   A=Constant3Vector() {constant=<rgba 0 0 1 0>}
   B=LinearInterpolate(
    A=@n57.N
    B=MaterialFunctionCall(BaseNormal=@n57.N, AdditionalNormal=@n58.N) {materialFunction=&BlendAngleCorrectedNormals}.Result
    Alpha=@n60.Alpha
   )
   Alpha=@n59.MaskOut Bottom Details
  )
  n63 = MakeMaterialAttributes!(
   BaseColor=Multiply(
    A=VectorParameter() {group=Rocks, parameterName=ColorIsland, defaultValue=<rgba 0.8906 0.4371 0.2481 0>}.RGB
    B=ShadingPathSwitch(Default=@n61, Deferred=@n61, Forward=@n60.Results, Mobile=@n60.Results)
   )
   Metallic
   Specular
   Roughness=ScalarParameter() {defaultValue=0.4, parameterName=R_Rocks}
   Anisotropy
   EmissiveColor
   Opacity
   OpacityMask
   Normal=MaterialFunctionCall(
    Normal=ShadingPathSwitch(Default=@n62, Deferred=@n62, Forward=@n57.N, Mobile=@n57.N)
    Flatness=ScalarParameter() {defaultValue=0.3, parameterName=Rock_NormalFlatten}
   ) {materialFunction=&FlattenNormal}.Result
   Tangent
   WorldPositionOffset
   SubsurfaceColor
   ClearCoat
   ClearCoatRoughness
   AmbientOcclusion
   Refraction
   CustomizedUVs_0
   CustomizedUVs_1
   CustomizedUVs_2
   CustomizedUVs_3
   CustomizedUVs_4
   CustomizedUVs_5
   CustomizedUVs_6
   CustomizedUVs_7
   PixelDepthOffset
   ShadingModel
   Displacement
  )
  n64 = Subtract(A=WorldPosition().XYZ, B=ObjectPositionWS())
  n65 = Constant3Vector() {constant=<rgba 0 0 -1900 0>}
  n66 = ScalarParameter() {defaultValue=4500, group=Emissive, parameterName=Light Radius}
  n67 = ScalarParameter() {defaultValue=-12, group=Emissive, parameterName=LightHardness}
  n68 = TextureCoordinate()
  n69 = Constant() {r=2}
  n70 = TextureSample(UVs=Rotator(Coordinate=Multiply(A=@n68, B=@n69), Time=Time())) {samplerType=SAMPLERTYPE_Grayscale, texture=&LowResBlurredNoise}
  n71 = SphereMask(A=@n64, B=@n65, Radius=@n66, Hardness=LinearInterpolate(A=@n67, B=Add(A=@n67) {constB=-30}, Alpha=@n70.R))
  n72 = Constant()
  n73 = Constant() {r=25}
  n74 = MaterialFunctionCall() {materialFunction=&MF_IslandWM_Game}
  n75 = Multiply(A=@n71, B=Multiply(A=LinearInterpolate(A=@n72, B=@n73, Alpha=@n74.GlowPartMask), B=@n70.R))
  n76 = SphereMask(
   A=@n64
   B=@n65
   Radius=Divide(A=@n66, B=ScalarParameter() {defaultValue=2.2, group=Emissive, parameterName=CoreRadiusScale})
   Hardness=Constant() {r=0.5}
  )
  n77 = MaterialFunctionCall(Value=Multiply(A=Time(), B=ScalarParameter() {defaultValue=1.4, parameterName=PulseSpeed_3})) {materialFunction=&LinearSine}
  n78 = ScalarParameter() {defaultValue=5, parameterName="Speed (S)"}
  n79 = ScalarParameter() {defaultValue=2, parameterName="Speed (S)_1"}
  n80 = ScalarParameter() {defaultValue=20, parameterName="Speed (S)_2"}
  n81 = ScalarParameter() {defaultValue=0.77, parameterName="Speed (S)_5"}
  n82 = Saturate(None=MaterialFunctionCall(
    A=MaterialFunctionCall(Speed=@n78) {materialFunction=&DebugTimeSine}.Result
    B=MaterialFunctionCall(Speed=@n79) {materialFunction=&DebugTimeSine}.Result
    C=MaterialFunctionCall(Speed=@n80) {materialFunction=&DebugTimeSine}.Result
    Alpha=MaterialFunctionCall(Speed=@n81) {materialFunction=&DebugTimeSine}.Result
   ) {materialFunction=&3ColorBlend}.Result)
  n83 = CollectionParameter() {collection=/Game/Athena/Prototype/Blueprints/Cube/Materials/MPC_Cube, parameterName=CubeGrowth_Unstable}
  n84 = Saturate(None=Add(
    A=LinearInterpolate(A=@n77."Rounded Linear Sine ", B=@n82, Alpha=@n83)
    B=ScalarParameter() {defaultValue=0.5, parameterName=CubeGrowth_Unstable_Minimum}
   ))
  n85 = LinearInterpolate(
   A=VectorParameter() {group=Emissive, parameterName=ColorFade, defaultValue=<rgba 0.9 0.1009 0.8856 1>}.RGB
   B=VectorParameter() {group=Emissive, parameterName=ColorCore, defaultValue=<rgba 4 0.2708 3.971 1>}.RGB
   Alpha=StaticSwitchParameter(True=Multiply(A=@n76, B=@n84), False=@n76) {parameterName=UseUnstableFlash}
  )
  n86 = LinearInterpolate(Alpha=MaterialFunctionCall() {materialFunction=&MF_IslandWM_Game}.GlowPartMask) {constA=0.5}
  n87 = Multiply(A=TextureCoordinate(), B=Constant() {r=12})
  n88 = MaterialFunctionCall() {materialFunction=&MF_IslandWM_Game}
  n89 = Add(
   A=Constant() {r=9.3422}
   B=Multiply(
    A=MaterialFunctionCall() {materialFunction=&MF_IslandRocks_Game}.A
    B=OneMinus(None=@n88.GlowPartMask)
   )
  )
  n90 = Constant() {r=-0.05}
  n91 = Multiply(
   A=@n86
   B=TextureSample(UVs=BumpOffset(Coordinate=@n87, Height=@n89, HeightRatioInput=@n90) {referencePlane=-0.25}) {samplerType=SAMPLERTYPE_Masks, texture=/Game/Athena/Prototype/Blueprints/Cube/Materials/T_DFHeightMask_Tile}.G
  )
  n92 = Multiply(A=@n85, B=Multiply(A=@n76, B=ShadingPathSwitch(Default=@n91, Deferred=@n91, Forward=@n86, Mobile=@n86)))
  n93 = SetMaterialAttributes!(MaterialAttributes=@n63, Emissive Color=Multiply(A=Add(A=@n71, B=@n75), B=Add(A=@n85, B=@n92))) {attributeSetTypes [#6]}
  n94 = GetMaterialAttributes(None=@n93) {attributeGetTypes [#2]}
  n95 = SetMaterialAttributes!(
   MaterialAttributes=@n94.MaterialAttributes
   Normal=MaterialFunctionCall(
    BaseNormal=@n94.Normal
    AdditionalNormal=TextureSampleParameter2D() {parameterName=N, samplerType=SAMPLERTYPE_Normal, texture=&T_IslandTop_N}.RGB
   ) {materialFunction=&BlendAngleCorrectedNormals}.Result
  ) {attributeSetTypes [#2]}
  MaterialFunctionCall(MA In=StaticSwitchParameter(True=BlendMaterialAttributes(A=@n63, B=@n51, Alpha=@n55), False=ShadingPathSwitch(Default=@n95, Deferred=@n95, Forward=@n93, Mobile=@n93)) {parameterName=TopGround/NoEmissive}) {materialFunction=&MF_CreativeMoveTool}
  Multiply(A=@n5.A)
 ]
}
